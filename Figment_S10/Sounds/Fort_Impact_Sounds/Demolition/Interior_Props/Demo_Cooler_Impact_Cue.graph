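# Sound cue for cooler interior-prop impacts, filed under demolition impact sounds.
# A Random node picks one of three equally weighted cooler impact waves, then a Modulator varies its pitch slightly.
# Playback uses small melee weapon attenuation, a duck-default impact sound class and a shared object-impact concurrency set, named for a four-voice cap.
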SoundCue /Figment_S10/Sounds/Fort_Impact_Sounds/Demolition/Interior_Props/Demo_Cooler_Impact_Cue
root=Modulator_2
nodes [
 {id=Modulator_2, children [Random_2], properties {pitchMax=1.07, pitchMin=0.92}}
 {id=Random_2, children [WavePlayer_12, WavePlayer_13, WavePlayer_14], properties {weights [1,1,1]}}
 {id=WavePlayer_12, properties {soundWaveAssetPtr=/Figment_S10/Sounds/Fort_Impact_Sounds/Demolition/Cooler_Impact02}}
 {id=WavePlayer_13, properties {soundWaveAssetPtr=/Figment_S10/Sounds/Fort_Impact_Sounds/Demolition/Cooler_Impact03}}
 {id=WavePlayer_14, properties {soundWaveAssetPtr=/Figment_S10/Sounds/Fort_Impact_Sounds/Demolition/Cooler_Impact04}}
]
properties {
 attenuationSettings=&Fort_Weapon_Melee_Sml_Att
 maxDistance=2800
 soundClassObject=/Game/Sounds/FortSoundClassesAndModes/Fort_ImpactSounds_DuckDefault_PSM
 volumeMultiplier=1
 concurrencyOverrides {resolutionRule=PreventNew}
 concurrencySet [&CONC_Object_Impacts_FourMax]
}
